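; Quest signpost prop for a Fortnite treasure-map (Hidden Stars) objective that changes to a treasure map look once completed.
; On objective completion, a timeline slightly scales the signpost mesh along Y and Z, then swaps its material to the treasure map material when finished.
; When the quest initializes as not valid, it checks objective completion for the first local player controller and applies that material directly, otherwise printing a debug message.
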(event EventBeginPlay
  (|CallParentFunction) ; This node is disabled and will not be called.\nDrag off pins to build functionality.
)

(event Collision|EventActorBeginOverlap
  (params (OtherActor :type Actor%20Object%20Reference))
  (|CallParentFunction OtherActor) ; This node is disabled and will not be called.\nDrag off pins to build functionality.
)

(event EventTick
  (params (DeltaSeconds :type Float%20%28single-precision%29))
  (|CallParentFunction DeltaSeconds) ; This node is disabled and will not be called.\nDrag off pins to build functionality.
)

(event EventThisQuestObjectiveComplete
  (params (IneractingPawn :type Fort%20Pawn%20Object%20Reference))
  (bind _sm_athena_treasure_signpost (Variables|Default|GetSM_Athena_Treasure_Signpost))
  (|CallParentFunction IneractingPawn)
  (bind (_direction _newtrack_0) (Timeline|Timeline_0 PlayFromStart)
    (:Update
      (Transformation|SetRelativeScale3D _sm_athena_treasure_signpost (Math|Vector|Lerp(Vector) "1.000000,1.000000,1.000000" "1.000000,1.100000,1.100000" _newtrack_0)))
    (:Finished
      (Rendering|Material|SetMaterial _sm_athena_treasure_signpost 0 (Variables|Default|GetTreasureMapMaterial)))))

(event EventQuestInitializedandNotValid
  (|CallParentFunction)
  (bind (_questisvalid _questcompleted _objectivecompleted) (Default|AthenaCheckQuestandObjectiveCompletion (Utilities|Array|Get(acopy) (Fortnite|GetLocalFortPlayerControllers true)) (Variables|QuestStuff|GetQuestItem) (Variables|Default|GetQuestBackendObjectiveName)))
  (if _objectivecompleted
    (Rendering|Material|SetMaterial (Variables|Default|GetSM_Athena_Treasure_Signpost) 0 (Variables|Default|GetTreasureMapMaterial))
    (else
      (Development|PrintString "Obj Not completed" false false "(R=1.000000,G=0.024273,B=0.000000,A=1.000000)" 1e+01))))

(fn ConstructionScript
  (|CallParentFunction))
